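# Physics asset for the rear cloth of the male medium Dragon Ninja T1 player body, modelling the flap with rigid bodies.
# Left and right chains of two flat boxes hang from small driver capsules, which link through dyn_cloth_rear to spine_01.
# Cross constraints tie the two sides together at each level, so they act as one panel.
# Pelvis, lower spine and thigh capsules supply body shapes, joined by angularly free constraints.
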PhysicsAsset /Game/Characters/Player/Male/Medium/Bodies/M_MED_Dragon_Ninja/Meshes/M_MED_Dragon_Ninja_T1_ClothRear_Physics
bodies {
 driver_rearcloth_01_l [{shapeType=Capsule, length=0.505, radius=0.505, rotation=<rot 0 -90 0>}]
 driver_rearcloth_01_r [{shapeType=Capsule, length=0.505, radius=0.505, rotation=<rot 0 -90 0>}]
 dyn_cloth_rear [{shapeType=Capsule, length=0.505, radius=0.505, rotation=<rot 0 -90 0>}]
 dyn_rearcloth_01_l [{shapeType=Box, extentX=11.5, extentY=1.5, extentZ=11.5, center=<4.3377 0 0.5929>, rotation=<rot 0 90 90>}]
 dyn_rearcloth_01_r [{shapeType=Box, extentX=11.5, extentY=1.5, extentZ=11.5, center=<-4.4916 0 0.0785>, rotation=<rot 0 90 90>}]
 dyn_rearcloth_02_l [{shapeType=Box, extentX=11.5, extentY=1.5, extentZ=11.5, center=<4.3781 0 0.0001>, rotation=<rot 0 90 90>}]
 dyn_rearcloth_02_r [{shapeType=Box, extentX=11.5, extentY=1.5, extentZ=11.5, center=<-4.4396 0 0.6859>, rotation=<rot 0 90 90>}]
 pelvis [{shapeType=Capsule, length=12.5, radius=11, center=<-5.9463 0.7666 0>}]
 spine_01 [{shapeType=Capsule, length=12.1464, radius=4.3382, center=<9.474 -2.3594 0>, rotation=<rot 0 -90 0>}]
 thigh_l [{shapeType=Capsule, length=19, radius=9.75, center=<-13.6517 1.0179 0.1294>, rotation=<rot 88.4923 -1.4669 0.9938>}]
 thigh_r [{shapeType=Capsule, length=19.25, radius=9.75, center=<13.6944 -0.8851 0.2368>, rotation=<rot 87.6961 -3.4876 0.3899>}]
}
constraints [
 {
  bone1Name=spine_01
  bone2Name=pelvis
  swing1LimitDegrees=45
  swing1Motion=Free
  swing2LimitDegrees=45
  swing2Motion=Free
  twistLimitDegrees=45
  twistMotion=Free
 }
 {
  bone1Name=thigh_l
  bone2Name=pelvis
  swing1LimitDegrees=45
  swing1Motion=Free
  swing2LimitDegrees=45
  swing2Motion=Free
  twistLimitDegrees=45
  twistMotion=Free
 }
 {
  bone1Name=thigh_r
  bone2Name=pelvis
  swing1LimitDegrees=45
  swing1Motion=Free
  swing2LimitDegrees=45
  swing2Motion=Free
  twistLimitDegrees=45
  twistMotion=Free
 }
 {
  bone1Name=driver_rearcloth_01_l
  bone2Name=dyn_cloth_rear
  swing1LimitDegrees=45
  swing1Motion=Free
  swing2LimitDegrees=45
  swing2Motion=Free
  twistLimitDegrees=45
  twistMotion=Free
 }
 {
  bone1Name=driver_rearcloth_01_r
  bone2Name=dyn_cloth_rear
  swing1LimitDegrees=45
  swing1Motion=Free
  swing2LimitDegrees=45
  swing2Motion=Free
  twistLimitDegrees=45
  twistMotion=Free
 }
 {bone1Name=dyn_rearcloth_01_l, bone2Name=driver_rearcloth_01_l, swing1LimitDegrees=45, swing2LimitDegrees=45, twistLimitDegrees=25}
 {bone1Name=dyn_rearcloth_02_l, bone2Name=dyn_rearcloth_01_l, swing1LimitDegrees=45, swing2LimitDegrees=45, twistLimitDegrees=25}
 {bone1Name=dyn_rearcloth_01_r, bone2Name=driver_rearcloth_01_r, swing1LimitDegrees=45, swing2LimitDegrees=45, twistLimitDegrees=25}
 {bone1Name=dyn_rearcloth_02_r, bone2Name=dyn_rearcloth_01_r, swing1LimitDegrees=45, swing2LimitDegrees=45, twistLimitDegrees=25}
 {bone1Name=dyn_rearcloth_01_r, bone2Name=dyn_rearcloth_01_l, swing1LimitDegrees=45, swing2LimitDegrees=45, twistLimitDegrees=45}
 {bone1Name=dyn_rearcloth_02_r, bone2Name=dyn_rearcloth_02_l, swing1LimitDegrees=45, swing2LimitDegrees=45, twistLimitDegrees=45}
 {
  bone1Name=dyn_cloth_rear
  bone2Name=spine_01
  swing1LimitDegrees=45
  swing1Motion=Free
  swing2LimitDegrees=45
  swing2Motion=Free
  twistLimitDegrees=45
  twistMotion=Free
 }
]
