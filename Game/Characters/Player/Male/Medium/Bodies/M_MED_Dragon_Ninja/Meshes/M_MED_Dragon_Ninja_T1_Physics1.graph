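# Physics asset for the Dragon Ninja medium male player body that sets up a two-segment physics loincloth on each side.
# Pelvis and thigh capsules approximate the lower body, while tiny driver capsules anchor the box-shaped loincloth chains.
# Each chain has limited swing and twist from driver through both segments.
# Thighs and drivers join the pelvis with free rotation, and free-rotation cross-constraints tie left and right cloth segments together at both levels.
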PhysicsAsset /Game/Characters/Player/Male/Medium/Bodies/M_MED_Dragon_Ninja/Meshes/M_MED_Dragon_Ninja_T1_Physics1
bodies {
 driver_loincloth_01_l [{shapeType=Capsule, length=0.505, radius=0.505, center=<-1.8737 0 0.0003>, rotation=<rot 0 -90 0>}]
 driver_loincloth_01_r [{shapeType=Capsule, length=0.505, radius=0.505, center=<1.9157 0 0>, rotation=<rot 0 -90 0>}]
 dyn_loincloth_01_l [{shapeType=Box, extentX=14.4603, extentY=1.7077, extentZ=10.6481, center=<11.3134 0.6282 0.3499>, rotation=<rot 0 0 90>}]
 dyn_loincloth_01_r [{shapeType=Box, extentX=14.6377, extentY=2.2029, extentZ=10.281, center=<-11.3576 -0.7522 -0.0736>, rotation=<rot 0 0 90>}]
 dyn_loincloth_02_l [{shapeType=Box, extentX=14.0357, extentY=1.9977, extentZ=11.985, center=<10.4239 0.752 0.6423>, rotation=<rot 0 0 90>}]
 dyn_loincloth_02_r [{shapeType=Box, extentX=14.2774, extentY=1.8389, extentZ=12.2451, center=<-10.3547 -0.7606 -0.412>, rotation=<rot 0 0 90>}]
 pelvis [{shapeType=Capsule, length=14.5211, radius=19.8693, center=<-11.7455 -7.264 0.5183>, rotation=<rot 3.1579 -92.9658 179.9173>}]
 thigh_l [{shapeType=Capsule, length=27.75, radius=11, center=<-13.9687 0.001 0>, rotation=<rot 90 0 0>}]
 thigh_r [{shapeType=Capsule, length=27.75, radius=11, center=<14.2476 -0.0009 0.6797>, rotation=<rot 90 0 0>}]
}
constraints [
 {bone1Name=dyn_loincloth_01_l, bone2Name=driver_loincloth_01_l, swing1LimitDegrees=40, swing2LimitDegrees=50, twistLimitDegrees=55}
 {bone1Name=dyn_loincloth_01_r, bone2Name=driver_loincloth_01_r, swing1LimitDegrees=40, swing2LimitDegrees=50, twistLimitDegrees=55}
 {
  bone1Name=thigh_l
  bone2Name=pelvis
  swing1LimitDegrees=45
  swing1Motion=Free
  swing2LimitDegrees=45
  swing2Motion=Free
  twistLimitDegrees=45
  twistMotion=Free
 }
 {
  bone1Name=thigh_r
  bone2Name=pelvis
  swing1LimitDegrees=45
  swing1Motion=Free
  swing2LimitDegrees=45
  swing2Motion=Free
  twistLimitDegrees=45
  twistMotion=Free
 }
 {
  bone1Name=driver_loincloth_01_l
  bone2Name=pelvis
  swing1LimitDegrees=45
  swing1Motion=Free
  swing2LimitDegrees=45
  swing2Motion=Free
  twistLimitDegrees=45
  twistMotion=Free
 }
 {
  bone1Name=driver_loincloth_01_r
  bone2Name=pelvis
  swing1LimitDegrees=45
  swing1Motion=Free
  swing2LimitDegrees=45
  swing2Motion=Free
  twistLimitDegrees=45
  twistMotion=Free
 }
 {bone1Name=dyn_loincloth_02_l, bone2Name=dyn_loincloth_01_l, swing1LimitDegrees=40, swing2LimitDegrees=50, twistLimitDegrees=55}
 {bone1Name=dyn_loincloth_02_r, bone2Name=dyn_loincloth_01_r, swing1LimitDegrees=40, swing2LimitDegrees=50, twistLimitDegrees=55}
 {
  bone1Name=dyn_loincloth_01_r
  bone2Name=dyn_loincloth_01_l
  swing1LimitDegrees=45
  swing1Motion=Free
  swing2LimitDegrees=45
  swing2Motion=Free
  twistLimitDegrees=45
  twistMotion=Free
 }
 {
  bone1Name=dyn_loincloth_02_r
  bone2Name=dyn_loincloth_02_l
  swing1LimitDegrees=45
  swing1Motion=Free
  swing2LimitDegrees=45
  swing2Motion=Free
  twistLimitDegrees=45
  twistMotion=Free
 }
]
