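# Animation Blueprint for the Dragon Ninja player body, adding secondary motion to shoulder armor, sleeve, loincloth and belts.
# It copies the attached parent's pose, then branches through cached poses and bool blends for cloth, skydiving and front-end idle.
# Bone offsets weighted by arm, ADS and skydive alphas, plus upper-arm Pose Drivers, correct the armor.
# Constraints, Anim Dynamics and RigidBody move the cloth, with `disable_custom_cloth` switching between setups.
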AnimBlueprint /Game/Characters/Player/Male/Medium/Bodies/M_MED_Dragon_Ninja/Meshes/M_MED_Dragon_Ninja_T1_AnimBP
target_skeleton=&/M_MED_Dragon_Ninja_T1_Skeleton
graphs [
 AnimGraph {nodes [
   n1 = Pose Driver - Source: upperarm_l!(SourcePose:Pose Link Structure=Use cached pose 'StartCache'()) -> (Pose:Pose Link Structure) {
    is=title AnimGraphNode_PoseDriver
    poseAsset=&/M_MED_Dragon_Ninja_ArmSkyDiving_PoseAsset
    onlyDriveBones [{boneName=dyn_shoulder_armor_LOCALSPACE_l}, {boneName=dyn_shoulder_armor_driver_l}, {boneName=dyn_shoulder_armor_driver_B_l}]
    poseTargets [
     {
      bApplyCustomCurve=false
      bIsHidden=false
      distanceMethod=DefaultMethod
      drivenName=Pose_14
      functionType=DefaultFunction
      targetScale=1
      boneTransforms [
       {targetRotation=<rot -64.7854 -3.089 6.0959>, targetTranslation=<15.5044 0.0034 -0.0112>}
       {targetRotation=<rot -85.1917 -138.5627 143.0621>, targetTranslation=<4.3543 -0.148 -2.3189>}
      ]
      customCurve {defaultValue=3.4028234663852886e+38, postInfinityExtrap=RCCE_Constant, preInfinityExtrap=RCCE_Constant, keys []}
      targetRotation=<rot 0 0 0>
     }
     {
      bApplyCustomCurve=false
      bIsHidden=false
      distanceMethod=DefaultMethod
      drivenName=Pose_15
      functionType=DefaultFunction
      targetScale=0.4
      boneTransforms [
       {targetRotation=<rot 37.8457 1.1504 -70.5749>, targetTranslation=<15.5044 0.0034 -0.0112>}
       {targetRotation=<rot -73.7168 22.625 -19.5329>, targetTranslation=<4.3543 -0.148 -2.3189>}
      ]
      customCurve {defaultValue=3.4028234663852886e+38, postInfinityExtrap=RCCE_Constant, preInfinityExtrap=RCCE_Constant, keys []}
      targetRotation=<rot 0 0 0>
     }
     {
      bApplyCustomCurve=false
      bIsHidden=false
      distanceMethod=DefaultMethod
      drivenName=Pose_16
      functionType=DefaultFunction
      targetScale=0.65
      boneTransforms [
       {targetRotation=<rot -8.8448 -22.4275 -73.1109>, targetTranslation=<15.5044 0.0034 -0.0112>}
       {targetRotation=<rot -64.8915 15.9083 -11.0855>, targetTranslation=<4.3543 -0.148 -2.3189>}
      ]
      customCurve {defaultValue=3.4028234663852886e+38, postInfinityExtrap=RCCE_Constant, preInfinityExtrap=RCCE_Constant, keys []}
      targetRotation=<rot 0 0 0>
     }
    ]
    rBFParams {distanceMethod=Quaternion, function=Linear, radius=250}
    sourceBones [{boneName=upperarm_l}, {boneName=clavicle_l}]
   }
   n2 = Pose Driver - Source: upperarm_r!(SourcePose:Pose Link Structure=@n1) -> (Pose:Pose Link Structure) {
    is=title AnimGraphNode_PoseDriver
    poseAsset=&M_MED_Dragon_Ninja_ArmSkyDiving_PoseAsset
    onlyDriveBones [{boneName=dyn_shoulder_armor_LOCALSPACE_r}, {boneName=dyn_shoulder_armor_driver_r}, {boneName=dyn_shoulder_armor_driver_B_r}]
    poseTargets [
     {
      bApplyCustomCurve=false
      bIsHidden=false
      distanceMethod=DefaultMethod
      drivenName=Pose_14
      functionType=DefaultFunction
      targetScale=1
      boneTransforms [{targetRotation=<rot -64.7855 -3.089 6.0959>, targetTranslation=<-15.4864 0.0027 0.0304>}]
      customCurve {defaultValue=3.4028234663852886e+38, postInfinityExtrap=RCCE_Constant, preInfinityExtrap=RCCE_Constant, keys []}
      targetRotation=<rot 0 0 0>
     }
     {
      bApplyCustomCurve=false
      bIsHidden=false
      distanceMethod=DefaultMethod
      drivenName=Pose_15
      functionType=DefaultFunction
      targetScale=0.4
      boneTransforms [{targetRotation=<rot 31.4016 -0.4748 -74.8269>, targetTranslation=<-15.4864 0.0027 0.0304>}]
      customCurve {defaultValue=3.4028234663852886e+38, postInfinityExtrap=RCCE_Constant, preInfinityExtrap=RCCE_Constant, keys []}
      targetRotation=<rot 0 0 0>
     }
     {
      bApplyCustomCurve=false
      bIsHidden=false
      distanceMethod=DefaultMethod
      drivenName=Pose_16
      functionType=DefaultFunction
      targetScale=0.65
      boneTransforms [{targetRotation=<rot -2.4503 -22.8264 -75.9646>, targetTranslation=<-15.4864 0.0027 0.0304>}]
      customCurve {defaultValue=3.4028234663852886e+38, postInfinityExtrap=RCCE_Constant, preInfinityExtrap=RCCE_Constant, keys []}
      targetRotation=<rot 0 0 0>
     }
    ]
    rBFParams {distanceMethod=Quaternion, function=Linear, radius=250}
    sourceBones [{boneName=upperarm_r}]
   }
   Blend Poses by bool(BlendPose_0=@n2, BlendPose_1=Use cached pose 'StartCache'(), bActiveValue=Get bIsSkyDiving()) {blendType=Linear}
   "Constraint - dyn_belt_01_r (P)"(Alpha=1, ConstraintWeights_0=1) {
    alphaBoolBlend {blendOption=Linear}
    boneToModify {boneName=dyn_belt_01_r}
    constraintSetup [{targetBone {boneName=pelvis}}]
    constraintWeights [1]
   }
   "Transform (Modify) Bone - Bone: driver_dyn_belt_01_r"(Alpha=1, Translation="-1,0,0", Rotation="-28,0,-1", Scale="1,1,1") {
    rotationMode=BMM_Additive
    rotationSpace=BCS_BoneSpace
    translationMode=BMM_Additive
    translationSpace=BCS_BoneSpace
    alphaBoolBlend {blendOption=Linear}
    boneToModify {boneName=driver_dyn_belt_01_r}
    rotation {pitch=-28, roll=-1}
    translation {x=-1}
   }
   n3 = Get skydive_parachute_alpha() -> (skydive_parachute_alpha:double)
   n4 = "Transform (Modify) Bone - Bone: dyn_shoulder_armor_driver_B_l"(
    ComponentPose=Local To Component(LocalPose=Use cached pose 'PreSkyDiving'())
    Alpha=@n3
    Translation="0,0,0"
    Rotation="-37.862,9.324,-13.767"
    Scale="1,1,1"
   ) {
    rotationMode=BMM_Additive
    rotationSpace=BCS_BoneSpace
    alphaBoolBlend {blendOption=Linear}
    boneToModify {boneName=dyn_shoulder_armor_driver_B_l}
    rotation=<rot -37.862 9.3241 -13.7672>
   }
   n5 = "Transform (Modify) Bone - Bone: dyn_shoulder_armor_driver_B_r"(ComponentPose=@n4, Alpha=@n3, Translation="0,0,0", Rotation="-45,0.928,-7.337", Scale="1,1,1") {
    rotationMode=BMM_Additive
    rotationSpace=BCS_BoneSpace
    alphaBoolBlend {blendOption=Linear}
    boneToModify {boneName=dyn_shoulder_armor_driver_B_r}
    rotation=<rot -45 0.928 -7.337>
   }
   n6 = "Transform (Modify) Bone - Bone: dyn_shoulder_armor_driver_l"(ComponentPose=@n5, Alpha=@n3, Translation="0,0,0", Rotation="10.842,-5.508,-1.060", Scale="0.8,0.8,0.8") {
    rotationMode=BMM_Additive
    rotationSpace=BCS_BoneSpace
    alphaBoolBlend {blendOption=Linear}
    boneToModify {boneName=dyn_shoulder_armor_driver_l}
    rotation=<rot 10.842 -5.508 -1.06>
    scale=<0.8 0.8 0.8>
   }
   n7 = "Transform (Modify) Bone - Bone: dyn_shoulder_armor_driver_r"(ComponentPose=@n6, Alpha=@n3, Translation="0,0,0", Rotation="6.743,-2.626,1.919", Scale="0.8,0.8,0.8") {
    rotationMode=BMM_Additive
    rotationSpace=BCS_BoneSpace
    alphaBoolBlend {blendOption=Linear}
    boneToModify {boneName=dyn_shoulder_armor_driver_r}
    rotation=<rot 6.743 -2.626 1.919>
    scale=<0.8 0.8 0.8>
   }
   n8 = Get skydive_falling_alpha() -> (skydive_falling_alpha:double)
   n9 = "Transform (Modify) Bone - Bone: dyn_shoulder_armor_driver_B_l"(ComponentPose=@n7, Alpha=@n8, Translation="0,0,0", Rotation="-24,3,-15.272", Scale="1,1,1") {
    rotationMode=BMM_Additive
    rotationSpace=BCS_BoneSpace
    alphaBoolBlend {blendOption=Linear}
    boneToModify {boneName=dyn_shoulder_armor_driver_B_l}
    rotation=<rot -24 3 -15.272>
   }
   n10 = "Transform (Modify) Bone - Bone: dyn_shoulder_armor_driver_B_r"(ComponentPose=@n9, Alpha=@n8, Translation="0,0,0", Rotation="-23.892,4.142,-23.347", Scale="1,1,1") {
    rotationMode=BMM_Additive
    rotationSpace=BCS_BoneSpace
    alphaBoolBlend {blendOption=Linear}
    boneToModify {boneName=dyn_shoulder_armor_driver_B_r}
    rotation=<rot -23.892 4.142 -23.3468>
   }
   n11 = "Transform (Modify) Bone - Bone: dyn_shoulder_armor_driver_l"(ComponentPose=@n10, Alpha=@n8, Translation="0,0,-3", Rotation="8,20,10", Scale="0.8,0.8,0.8") {
    rotationMode=BMM_Additive
    rotationSpace=BCS_BoneSpace
    translationMode=BMM_Additive
    translationSpace=BCS_BoneSpace
    alphaBoolBlend {blendOption=Linear}
    boneToModify {boneName=dyn_shoulder_armor_driver_l}
    rotation=<rot 8 20 10>
    scale=<0.8 0.8 0.8>
    translation {z=-3}
   }
   n12 = "Transform (Modify) Bone - Bone: dyn_shoulder_armor_driver_r"(ComponentPose=@n11, Alpha=@n8, Translation="0,0,3", Rotation="8,20,10", Scale="0.8,0.8,0.8") {
    rotationMode=BMM_Additive
    rotationSpace=BCS_BoneSpace
    translationMode=BMM_Additive
    translationSpace=BCS_BoneSpace
    alphaBoolBlend {blendOption=Linear}
    boneToModify {boneName=dyn_shoulder_armor_driver_r}
    rotation=<rot 8 20 10>
    scale=<0.8 0.8 0.8>
    translation {z=3}
   }
   n13 = Component To Local(ComponentPose=@n12)
   n14 = "Transform (Modify) Bone - Bone: dyn_shoulder_armor_LOCALSPACE_l"(
    ComponentPose=Local To Component(LocalPose=Use cached pose 'PreSkyDiving'())
    Alpha=1
    Translation="-8,0,3"
    Rotation="0,0,-65"
    Scale="1,1,1"
   ) {
    rotationMode=BMM_Additive
    rotationSpace=BCS_BoneSpace
    translationMode=BMM_Additive
    translationSpace=BCS_BoneSpace
    alphaBoolBlend {blendOption=Linear}
    boneToModify {boneName=dyn_shoulder_armor_LOCALSPACE_l}
    rotation {roll=-65}
    translation {x=-8, z=3}
   }
   n15 = Get skydive_umbrella_alpha() -> (skydive_umbrella_alpha:double)
   n16 = "Transform (Modify) Bone - Bone: dyn_shoulder_armor_driver_B_l"(ComponentPose=@n14, Alpha=@n15, Translation="0,0,5", Rotation="27,0.000,0", Scale="1,1,1") {
    rotationMode=BMM_Additive
    rotationSpace=BCS_BoneSpace
    translationMode=BMM_Additive
    translationSpace=BCS_BoneSpace
    alphaBoolBlend {blendOption=Linear}
    boneToModify {boneName=dyn_shoulder_armor_driver_B_l}
    rotation {pitch=27}
    translation {z=5}
   }
   n17 = "Transform (Modify) Bone - Bone: dyn_shoulder_armor_driver_B_r"(ComponentPose=@n16, Alpha=@n15, Translation="0,0,0", Rotation="-23.892,4.142,-23.347", Scale="1,1,1") {
    rotationMode=BMM_Additive
    rotationSpace=BCS_BoneSpace
    alphaBoolBlend {blendOption=Linear}
    boneToModify {boneName=dyn_shoulder_armor_driver_B_r}
    rotation=<rot -23.892 4.142 -23.3468>
   }
   n18 = "Transform (Modify) Bone - Bone: dyn_shoulder_armor_driver_l"(ComponentPose=@n17, Alpha=@n15, Translation="0,0,-3", Rotation="5,0,15", Scale="0.8,0.8,0.8") {
    rotationMode=BMM_Additive
    rotationSpace=BCS_BoneSpace
    translationMode=BMM_Additive
    translationSpace=BCS_BoneSpace
    alphaBoolBlend {blendOption=Linear}
    boneToModify {boneName=dyn_shoulder_armor_driver_l}
    rotation {pitch=5, roll=15}
    scale=<0.8 0.8 0.8>
    translation {z=-3}
   }
   n19 = "Transform (Modify) Bone - Bone: dyn_shoulder_armor_driver_r"(ComponentPose=@n18, Alpha=@n15, Translation="0,0,3", Rotation="8,20,10", Scale="0.8,0.8,0.8") {
    rotationMode=BMM_Additive
    rotationSpace=BCS_BoneSpace
    translationMode=BMM_Additive
    translationSpace=BCS_BoneSpace
    alphaBoolBlend {blendOption=Linear}
    boneToModify {boneName=dyn_shoulder_armor_driver_r}
    rotation=<rot 8 20 10>
    scale=<0.8 0.8 0.8>
    translation {z=3}
   }
   n20 = Component To Local(ComponentPose=@n19)
   n21 = Get skydiving_umbrella_active() -> (skydiving_umbrella_active:bool)
   n22 = "Transform (Modify) Bone - Bone: dyn_shoulder_offset_r"(
    ComponentPose=Local To Component(LocalPose=Blend Poses by bool(BlendPose_0=@n20, BlendPose_1=@n13, bActiveValue=@n21) {blendType=Linear})
    Alpha=Get arm_pitch_up_r() -> (arm_pitch_up_r:double)
    Translation="0.5,0,11"
    Rotation="35,0,0"
    Scale="0.75,0.75,0.75"
   ) {
    rotationMode=BMM_Additive
    rotationSpace=BCS_BoneSpace
    scaleMode=BMM_Additive
    scaleSpace=BCS_BoneSpace
    translationMode=BMM_Additive
    translationSpace=BCS_BoneSpace
    alphaBoolBlend {blendOption=Linear}
    alphaScaleBiasClamp {bInterpResult=true, interpSpeedDecreasing=30, interpSpeedIncreasing=30}
    boneToModify {boneName=dyn_shoulder_offset_r}
    rotation {pitch=35}
    scale=<0.75 0.75 0.75>
    translation {x=0.5, z=11}
   }
   n23 = Anim Dynamics - Bone: dyn_sleeve_tip(ComponentPose=@n22, Alpha=0.55, ExternalForce="0,0,0") {
    angularDampingOverride=0.9999
    angularSpringConstant=100
    bAngularSpring=true
    bGravityOverrideInSimSpace=true
    bOverrideAngularDamping=true
    gravityScale=0.33
    simulationSpace=RootRelative
    alphaBoolBlend {blendOption=Linear}
    boundBone {boneName=dyn_sleeve_tip}
    physicsBodyDefinitions [{
      sphereCollisionRadius=0
      boundBone {boneName=dyn_sleeve_tip}
      boxExtents=<30 30 20>
      constraintSetup {twistAxis=AxisZ, angularLimitsMax=<25 25 25>, angularLimitsMin=<-25 -25 -25>, angularTarget=<0 0 0>}
      localJointOffset=<0 0 -10>
     }]
    relativeSpaceBone {boneName=spine_05}
   }
   n24 = Component To Local(ComponentPose=@n23)
   n25 = Get arm_roll_back_r() -> (arm_roll_back_r:double)
   n26 = Get arm_roll_back_l() -> (arm_roll_back_l:double)
   n27 = "Transform (Modify) Bone - Bone: dyn_shoulder_armor_LOCALSPACE_l"(
    ComponentPose=Local To Component(LocalPose=Use cached pose 'PreSkyDiving'())
    Alpha=1
    Translation="2,0.000,1.5"
    Rotation="0,0,5"
    Scale="1,1,1"
   ) {
    rotationMode=BMM_Additive
    rotationSpace=BCS_BoneSpace
    scaleMode=BMM_Additive
    translationMode=BMM_Additive
    translationSpace=BCS_BoneSpace
    alphaBoolBlend {blendOption=Linear}
    boneToModify {boneName=dyn_shoulder_armor_LOCALSPACE_l}
    rotation {roll=5}
    translation {x=2, z=1.5}
   }
   n28 = "Transform (Modify) Bone - Bone: dyn_shoulder_bck_l"(ComponentPose=@n27, Alpha=1, Translation="0,0,0", Rotation="4.5,0.000,0.000", Scale="1,1,1") {
    rotationMode=BMM_Additive
    rotationSpace=BCS_BoneSpace
    scaleMode=BMM_Additive
    scaleSpace=BCS_BoneSpace
    translationMode=BMM_Additive
    translationSpace=BCS_BoneSpace
    alphaBoolBlend {blendOption=Linear}
    boneToModify {boneName=dyn_shoulder_bck_l}
    rotation {pitch=4.5}
   }
   n29 = "Transform (Modify) Bone - Bone: dyn_shoulder_frt_l"(ComponentPose=@n28, Alpha=1, Translation="1,0,0", Rotation="0,0,0", Scale="1,1,1") {
    rotationMode=BMM_Additive
    rotationSpace=BCS_BoneSpace
    scaleMode=BMM_Additive
    scaleSpace=BCS_BoneSpace
    translationMode=BMM_Additive
    translationSpace=BCS_BoneSpace
    alphaBoolBlend {blendOption=Linear}
    boneToModify {boneName=dyn_shoulder_frt_l}
    translation {x=1}
   }
   n30 = Get arm_roll_fwd_l() -> (arm_roll_fwd_l:double)
   n31 = "Transform (Modify) Bone - Bone: dyn_shoulder_armor_LOCALSPACE_l"(ComponentPose=@n29, Alpha=@n30, Translation="0,0,0", Rotation="0.000,0.000,-50", Scale="1,1,1") {
    rotationMode=BMM_Additive
    rotationSpace=BCS_BoneSpace
    scaleSpace=BCS_BoneSpace
    translationSpace=BCS_BoneSpace
    alphaBoolBlend {blendOption=Linear}
    alphaScaleBiasClamp {bInterpResult=true, interpSpeedDecreasing=30, interpSpeedIncreasing=30}
    boneToModify {boneName=dyn_shoulder_armor_LOCALSPACE_l}
    rotation {roll=-50}
   }
   n32 = "Transform (Modify) Bone - Bone: dyn_shoulder_bck_l"(ComponentPose=@n31, Alpha=@n30, Translation="0,0,0", Rotation="10,0.000,0.000", Scale="1,1,1") {
    rotationMode=BMM_Additive
    rotationSpace=BCS_BoneSpace
    scaleSpace=BCS_BoneSpace
    translationSpace=BCS_BoneSpace
    alphaBoolBlend {blendOption=Linear}
    alphaScaleBiasClamp {bInterpResult=true, interpSpeedDecreasing=30, interpSpeedIncreasing=30}
    boneToModify {boneName=dyn_shoulder_bck_l}
    rotation {pitch=10}
   }
   n33 = "Transform (Modify) Bone - Bone: dyn_shoulder_armor_LOCALSPACE_l"(ComponentPose=@n32, Alpha=@n26, Translation="0,0,0", Rotation="0.000,0.000,35", Scale="1,1,1") {
    rotationMode=BMM_Additive
    rotationSpace=BCS_BoneSpace
    scaleSpace=BCS_BoneSpace
    translationSpace=BCS_BoneSpace
    alphaBoolBlend {blendOption=Linear}
    alphaScaleBiasClamp {bInterpResult=true, interpSpeedDecreasing=30, interpSpeedIncreasing=30}
    boneToModify {boneName=dyn_shoulder_armor_LOCALSPACE_l}
    rotation {roll=35}
   }
   n34 = "Transform (Modify) Bone - Bone: dyn_shoulder_frt_l"(ComponentPose=@n33, Alpha=@n26, Translation="0,0,0", Rotation="5,0.000,0.000", Scale="1,1,1") {
    rotationMode=BMM_Additive
    rotationSpace=BCS_BoneSpace
    scaleSpace=BCS_BoneSpace
    translationSpace=BCS_BoneSpace
    alphaBoolBlend {blendOption=Linear}
    alphaScaleBiasClamp {bInterpResult=true, interpSpeedDecreasing=30, interpSpeedIncreasing=30}
    boneToModify {boneName=dyn_shoulder_frt_l}
    rotation {pitch=5}
   }
   n35 = Get arm_pitch_up_l() -> (arm_pitch_up_l:double)
   n36 = "Transform (Modify) Bone - Bone: dyn_shoulder_armor_driver_B_l"(ComponentPose=@n34, Alpha=@n35, Translation="0,0,0", Rotation="-17,0,0", Scale="1,1,1") {
    rotationMode=BMM_Additive
    rotationSpace=BCS_BoneSpace
    scaleSpace=BCS_BoneSpace
    translationSpace=BCS_BoneSpace
    alphaBoolBlend {blendOption=Linear}
    alphaScaleBiasClamp {bInterpResult=true, interpSpeedDecreasing=30, interpSpeedIncreasing=30}
    boneToModify {boneName=dyn_shoulder_armor_driver_B_l}
    rotation {pitch=-17}
   }
   n37 = "Transform (Modify) Bone - Bone: dyn_shoulder_armor_LOCALSPACE_l"(ComponentPose=@n36, Alpha=@n35, Translation="0,0,0", Rotation="0,0,23", Scale="1,1,1") {
    rotationMode=BMM_Additive
    rotationSpace=BCS_BoneSpace
    scaleSpace=BCS_BoneSpace
    translationSpace=BCS_BoneSpace
    alphaBoolBlend {blendOption=Linear}
    alphaScaleBiasClamp {bInterpResult=true, interpSpeedDecreasing=30, interpSpeedIncreasing=30}
    boneToModify {boneName=dyn_shoulder_armor_LOCALSPACE_l}
    rotation {roll=23}
   }
   n38 = "Transform (Modify) Bone - Bone: dyn_shoulder_armor_LOCALSPACE_r"(ComponentPose=@n37, Alpha=1, Translation="-2,0.000,-1.5", Rotation="0,0,5", Scale="1,1,1") {
    rotationMode=BMM_Additive
    rotationSpace=BCS_BoneSpace
    scaleMode=BMM_Additive
    translationMode=BMM_Additive
    translationSpace=BCS_BoneSpace
    alphaBoolBlend {blendOption=Linear}
    boneToModify {boneName=dyn_shoulder_armor_LOCALSPACE_r}
    rotation {roll=5}
    translation {x=-2, z=-1.5}
   }
   n39 = "Transform (Modify) Bone - Bone: dyn_shoulder_bck_r"(ComponentPose=@n38, Alpha=1, Translation="0,0,0", Rotation="4.5,0,0", Scale="1,1,1") {
    rotationMode=BMM_Additive
    rotationSpace=BCS_BoneSpace
    scaleMode=BMM_Additive
    scaleSpace=BCS_BoneSpace
    translationMode=BMM_Additive
    translationSpace=BCS_BoneSpace
    alphaBoolBlend {blendOption=Linear}
    boneToModify {boneName=dyn_shoulder_bck_r}
    rotation {pitch=4.5}
   }
   n40 = "Transform (Modify) Bone - Bone: dyn_shoulder_frt_r"(ComponentPose=@n39, Alpha=1, Translation="1,0,0", Rotation="0,0,0", Scale="1,1,1") {
    rotationMode=BMM_Additive
    rotationSpace=BCS_BoneSpace
    scaleMode=BMM_Additive
    scaleSpace=BCS_BoneSpace
    translationMode=BMM_Additive
    translationSpace=BCS_BoneSpace
    alphaBoolBlend {blendOption=Linear}
    boneToModify {boneName=dyn_shoulder_frt_r}
    translation {x=1}
   }
   n41 = Get arm_roll_fwd_r() -> (arm_roll_fwd_r:double)
   n42 = "Transform (Modify) Bone - Bone: dyn_shoulder_armor_LOCALSPACE_r"(ComponentPose=@n40, Alpha=@n41, Translation="0,0,0", Rotation="0.000,0.000,-50", Scale="1,1,1") {
    rotationMode=BMM_Additive
    rotationSpace=BCS_BoneSpace
    scaleSpace=BCS_BoneSpace
    translationSpace=BCS_BoneSpace
    alphaBoolBlend {blendOption=Linear}
    alphaScaleBiasClamp {bInterpResult=true, interpSpeedDecreasing=30, interpSpeedIncreasing=30}
    boneToModify {boneName=dyn_shoulder_armor_LOCALSPACE_r}
    rotation {roll=-50}
   }
   n43 = "Transform (Modify) Bone - Bone: dyn_shoulder_bck_r"(ComponentPose=@n42, Alpha=@n41, Translation="0,0,0", Rotation="10,0.000,0.000", Scale="1,1,1") {
    rotationMode=BMM_Additive
    rotationSpace=BCS_BoneSpace
    scaleSpace=BCS_BoneSpace
    translationSpace=BCS_BoneSpace
    alphaBoolBlend {blendOption=Linear}
    alphaScaleBiasClamp {bInterpResult=true, interpSpeedDecreasing=30, interpSpeedIncreasing=30}
    boneToModify {boneName=dyn_shoulder_bck_r}
    rotation {pitch=10}
   }
   n44 = "Transform (Modify) Bone - Bone: dyn_shoulder_armor_LOCALSPACE_r"(ComponentPose=@n43, Alpha=@n25, Translation="0,0,0", Rotation="0.000,0.000,35", Scale="1,1,1") {
    rotationMode=BMM_Additive
    rotationSpace=BCS_BoneSpace
    scaleSpace=BCS_BoneSpace
    translationSpace=BCS_BoneSpace
    alphaBoolBlend {blendOption=Linear}
    alphaScaleBiasClamp {bInterpResult=true, interpSpeedDecreasing=30, interpSpeedIncreasing=30}
    boneToModify {boneName=dyn_shoulder_armor_LOCALSPACE_r}
    rotation {roll=35}
   }
   n45 = "Transform (Modify) Bone - Bone: dyn_shoulder_frt_r"(ComponentPose=@n44, Alpha=@n25, Translation="0,0,0", Rotation="15,0.000,0.000", Scale="1,1,1") {
    rotationMode=BMM_Additive
    rotationSpace=BCS_BoneSpace
    scaleSpace=BCS_BoneSpace
    translationSpace=BCS_BoneSpace
    alphaBoolBlend {blendOption=Linear}
    alphaScaleBiasClamp {bInterpResult=true, interpSpeedDecreasing=30, interpSpeedIncreasing=30}
    boneToModify {boneName=dyn_shoulder_frt_r}
    rotation {pitch=15}
   }
   n46 = Get arm_pitch_up_r()
   n47 = "Transform (Modify) Bone - Bone: dyn_shoulder_armor_driver_B_r"(ComponentPose=@n45, Alpha=@n46, Translation="0,0,0", Rotation="-17,0,0", Scale="1,1,1") {
    rotationMode=BMM_Additive
    rotationSpace=BCS_BoneSpace
    scaleSpace=BCS_BoneSpace
    translationSpace=BCS_BoneSpace
    alphaBoolBlend {blendOption=Linear}
    alphaScaleBiasClamp {bInterpResult=true, interpSpeedDecreasing=30, interpSpeedIncreasing=30}
    boneToModify {boneName=dyn_shoulder_armor_driver_B_r}
    rotation {pitch=-17}
   }
   n48 = "Transform (Modify) Bone - Bone: dyn_shoulder_armor_LOCALSPACE_r"(ComponentPose=@n47, Alpha=@n46, Translation="0,0,0", Rotation="0,0,23", Scale="1,1,1") {
    rotationMode=BMM_Additive
    rotationSpace=BCS_BoneSpace
    scaleSpace=BCS_BoneSpace
    translationSpace=BCS_BoneSpace
    alphaBoolBlend {blendOption=Linear}
    alphaScaleBiasClamp {bInterpResult=true, interpSpeedDecreasing=30, interpSpeedIncreasing=30}
    boneToModify {boneName=dyn_shoulder_armor_LOCALSPACE_r}
    rotation {roll=23}
   }
   n49 = "Transform (Modify) Bone - Bone: dyn_shoulder_offset_r"(ComponentPose=@n48, Alpha=@n46, Translation="3,0,1", Rotation="-15,10,-5", Scale="1,1,1") {
    rotationMode=BMM_Additive
    rotationSpace=BCS_BoneSpace
    scaleSpace=BCS_BoneSpace
    translationMode=BMM_Additive
    translationSpace=BCS_BoneSpace
    alphaBoolBlend {blendOption=Linear}
    alphaScaleBiasClamp {bInterpResult=true, interpSpeedDecreasing=30, interpSpeedIncreasing=30}
    boneToModify {boneName=dyn_shoulder_offset_r}
    rotation=<rot -15 10 -5>
    translation {x=3, z=1}
   }
   n50 = "Transform (Modify) Bone - Bone: dyn_shoulder_offset_r"(ComponentPose=@n49, Alpha=Get arm_up_r() -> (arm_up_r:double), Translation="0,-0.482,2.666", Rotation="-23.738,0,0.262", Scale="0.7,0.7,0.7") {
    rotationMode=BMM_Additive
    rotationSpace=BCS_BoneSpace
    scaleMode=BMM_Additive
    scaleSpace=BCS_BoneSpace
    translationMode=BMM_Additive
    translationSpace=BCS_BoneSpace
    alphaBoolBlend {blendOption=Linear}
    alphaScaleBiasClamp {bInterpResult=true, interpSpeedDecreasing=30, interpSpeedIncreasing=30}
    boneToModify {boneName=dyn_shoulder_offset_r}
    rotation {pitch=-23.738, roll=0.262}
    scale=<0.7 0.7 0.7>
    translation {y=-0.482, z=2.666}
   }
   n51 = Get ads_alpha() -> (ads_alpha:double)
   n52 = "Transform (Modify) Bone - Bone: dyn_sleeve_tipB"(ComponentPose=@n50, Alpha=@n51, Translation="5,0,0", Rotation="0,0,0", Scale="0.55,0.55,0.55") {
    rotationMode=BMM_Additive
    rotationSpace=BCS_BoneSpace
    scaleMode=BMM_Additive
    scaleSpace=BCS_BoneSpace
    translationMode=BMM_Additive
    translationSpace=BCS_BoneSpace
    alphaBoolBlend {blendOption=Linear}
    alphaScaleBiasClamp {bInterpResult=true, interpSpeedDecreasing=30, interpSpeedIncreasing=30}
    boneToModify {boneName=dyn_sleeve_tipB}
    scale=<0.55 0.55 0.55>
    translation {x=5}
   }
   n53 = "Transform (Modify) Bone - Bone: upperarm_corrective_r"(ComponentPose=@n52, Alpha=@n51, Translation="0,0,0", Rotation="-65,-110,50", Scale="0.750,0.750,0.750") {
    rotationMode=BMM_Additive
    rotationSpace=BCS_BoneSpace
    scaleMode=BMM_Additive
    scaleSpace=BCS_BoneSpace
    translationMode=BMM_Additive
    translationSpace=BCS_BoneSpace
    alphaBoolBlend {blendOption=Linear}
    alphaScaleBiasClamp {bInterpResult=true, interpSpeedDecreasing=30, interpSpeedIncreasing=30}
    boneToModify {boneName=upperarm_corrective_r}
    rotation=<rot -65 -110 50>
    scale=<0.75 0.75 0.75>
   }
   n54 = Anim Dynamics - Bone: dyn_sleeve_tip(ComponentPose=@n53, Alpha=Get ads_tip_alpha() -> (ads_tip_alpha:double), ExternalForce="0,0,0") {
    angularDampingOverride=0.9999
    angularSpringConstant=100
    bAngularSpring=true
    bGravityOverrideInSimSpace=true
    bOverrideAngularDamping=true
    gravityScale=0.33
    simulationSpace=RootRelative
    alphaBoolBlend {blendOption=Linear}
    boundBone {boneName=dyn_sleeve_tip}
    physicsBodyDefinitions [{
      sphereCollisionRadius=0
      boundBone {boneName=dyn_sleeve_tip}
      boxExtents=<30 30 20>
      constraintSetup {twistAxis=AxisZ, angularLimitsMax=<25 25 25>, angularLimitsMin=<-25 -25 -25>, angularTarget=<0 0 0>}
      localJointOffset=<0 0 -10>
     }]
    relativeSpaceBone {boneName=spine_05}
   }
   n55 = Component To Local(ComponentPose=@n54)
   n56 = Get bIsSkyDiving()
   n57 = "Transform (Modify) Bone - Bone: dyn_shoulder_armor_LOCALSPACE_r"!(ComponentPose=Local To Component(LocalPose=Use cached pose 'PreSkyDiving'()), Alpha=1, Rotation="13.211,41.816,23.686") -> (Pose) {
    rotationMode=BMM_Additive
    alphaBoolBlend {blendOption=Linear}
    boneToModify {boneName=dyn_shoulder_armor_LOCALSPACE_r}
    rotation=<rot 13.211 41.816 23.686>
   }
   n58 = "Transform (Modify) Bone - Bone: dyn_shoulder_armor_LOCALSPACE_l"!(ComponentPose=@n57, Alpha=1, Rotation="-27.645,-59.907,-11.731") -> (Pose) {
    rotationMode=BMM_Additive
    alphaBoolBlend {blendOption=Linear}
    boneToModify {boneName=dyn_shoulder_armor_LOCALSPACE_l}
    rotation=<rot -27.645 -59.907 -11.731>
   }
   Output Pose(Result=Blend Poses by bool(
     BlendPose_0=Component To Local(ComponentPose=@n58)
     BlendPose_1=Blend Poses by bool(BlendPose_0=@n24, BlendPose_1=@n55, bActiveValue=@n56) {blendType=Linear}
     BlendTime_0=0
     BlendTime_1=0
     bActiveValue=Get FrontEnd_Idle_On() -> (FrontEnd_Idle_On:bool)
    ) {blendType=Linear}) {layerGroup=None}
   Save cached pose 'AnimCache'(Pose=Copy Pose From Mesh() {bUseAttachedParent=true}) {cacheName=AnimCache}
   n59 = "Constraint - driver_loincloth_01_l (P)"(ComponentPose=Local To Component(LocalPose=Use cached pose 'AnimCache'()), Alpha=1, ConstraintWeights_0=1) {
    alphaBoolBlend {blendOption=Linear}
    boneToModify {boneName=driver_loincloth_01_l}
    constraintSetup [{targetBone {boneName=thigh_l}}]
    constraintWeights [1]
   }
   n60 = "Constraint - driver_loincloth_01_r (PP)"!(ComponentPose=@n59, Alpha=1, ConstraintWeights_0=0.8, ConstraintWeights_1:float=0.2) -> (Pose) {
    alphaBoolBlend {blendOption=Linear}
    boneToModify {boneName=driver_loincloth_01_r}
    constraintSetup [{targetBone {boneName=thigh_r}}, {targetBone {boneName=spine_01}}]
    constraintWeights [1,1]
   }
   n61 = "Transform (Modify) Bone - Bone: driver_loincloth_01_l"(ComponentPose=@n60, Alpha=1, Translation="5,0,0", Rotation="0,11,0", Scale="1,1,1") {
    rotationMode=BMM_Additive
    rotationSpace=BCS_BoneSpace
    translationMode=BMM_Additive
    translationSpace=BCS_BoneSpace
    alphaBoolBlend {blendOption=Linear}
    boneToModify {boneName=driver_loincloth_01_l}
    rotation {yaw=11}
    translation {x=5}
   }
   n62 = Anim Dynamics - Bone: dyn_loincloth_02_l(ComponentPose=@n61, Alpha=1, ExternalForce="0,0,0") {
    angularDampingOverride=0.9995
    angularSpringConstant=300
    bAngularSpring=true
    bGravityOverrideInSimSpace=true
    bOverrideAngularDamping=true
    gravityScale=0.33
    simulationSpace=RootRelative
    alphaBoolBlend {blendOption=Linear}
    boundBone {boneName=dyn_loincloth_02_l}
    physicsBodyDefinitions [{
      sphereCollisionRadius=0
      boundBone {boneName=dyn_loincloth_02_l}
      boxExtents=<10 10 10>
      constraintSetup {angularLimitsMax=<1 1 15>, angularLimitsMin=<-25 -20 -15>, angularTarget=<0 0 0>}
      localJointOffset=<1 0 0>
     }]
   }
   n63 = Anim Dynamics - Bone: dyn_loincloth_03A_l(ComponentPose=@n62, Alpha=1, ExternalForce="0,0,0") {
    angularDampingOverride=0.9995
    angularSpringConstant=300
    bAngularSpring=true
    bGravityOverrideInSimSpace=true
    bOverrideAngularDamping=true
    gravityScale=0.33
    simulationSpace=RootRelative
    alphaBoolBlend {blendOption=Linear}
    boundBone {boneName=dyn_loincloth_03A_l}
    physicsBodyDefinitions [{
      sphereCollisionRadius=0
      boundBone {boneName=dyn_loincloth_03A_l}
      boxExtents=<10 10 10>
      constraintSetup {twistAxis=AxisZ, angularLimitsMax=<1 20 15>, angularLimitsMin=<-25 -20 -1>, angularTarget=<0 0 0>}
      localJointOffset=<0 0 -5>
     }]
   }
   n64 = Anim Dynamics - Bone: dyn_loincloth_02_r(ComponentPose=@n63, Alpha=1, ExternalForce="0,0,0") {
    angularDampingOverride=0.9995
    angularSpringConstant=300
    bAngularSpring=true
    bGravityOverrideInSimSpace=true
    bOverrideAngularDamping=true
    gravityScale=0.33
    simulationSpace=RootRelative
    alphaBoolBlend {blendOption=Linear}
    boundBone {boneName=dyn_loincloth_02_r}
    physicsBodyDefinitions [{
      sphereCollisionRadius=0
      boundBone {boneName=dyn_loincloth_02_r}
      boxExtents=<10 10 10>
      constraintSetup {angularLimitsMax=<25 1 15>, angularLimitsMin=<1 -20 -15>, angularTarget=<0 0 0>}
      localJointOffset=<-1 0 0>
     }]
   }
   n65 = Anim Dynamics - Bone: dyn_loincloth_03A_r(ComponentPose=@n64, Alpha=1, ExternalForce="0,0,0") {
    angularDampingOverride=0.9995
    angularSpringConstant=300
    bAngularSpring=true
    bGravityOverrideInSimSpace=true
    bOverrideAngularDamping=true
    gravityScale=0.33
    simulationSpace=RootRelative
    alphaBoolBlend {blendOption=Linear}
    boundBone {boneName=dyn_loincloth_03A_r}
    physicsBodyDefinitions [{
      sphereCollisionRadius=0
      boundBone {boneName=dyn_loincloth_03A_r}
      boxExtents=<10 10 10>
      constraintSetup {twistAxis=AxisZ, angularLimitsMax=<1 20 15>, angularLimitsMin=<-25 -20 -1>, angularTarget=<0 0 0>}
      localJointOffset=<0 0 -5>
     }]
   }
   n66 = "Constraint - driver_dyn_belt_01_r (PP)"!(ComponentPose=@n65, Alpha=1, ConstraintWeights_0=0.7, ConstraintWeights_1:float=0.3) -> (Pose) {
    alphaBoolBlend {blendOption=Linear}
    boneToModify {boneName=driver_dyn_belt_01_r}
    constraintSetup [{targetBone {boneName=Flap_Front_Parent_Space}}, {targetBone {boneName=pelvis}}]
    constraintWeights [1,1]
   }
   n67 = "Constraint - parent_dyn_belt_01_l (P)"(ComponentPose=@n66, Alpha=1, ConstraintWeights_0=1) {
    alphaBoolBlend {blendOption=Linear}
    boneToModify {boneName=parent_dyn_belt_01_l}
    constraintSetup [{targetBone {boneName=dyn_loincloth_01_l}}]
    constraintWeights [1]
   }
   n68 = "Transform (Modify) Bone - Bone: driver_dyn_belt_01_l"(ComponentPose=@n67, Alpha=1, Translation="0,0,-5", Rotation="0,0,0", Scale="0.7,0.7,0.7") {
    rotationMode=BMM_Additive
    rotationSpace=BCS_BoneSpace
    scaleSpace=BCS_BoneSpace
    translationMode=BMM_Additive
    translationSpace=BCS_BoneSpace
    alphaBoolBlend {blendOption=Linear}
    boneToModify {boneName=driver_dyn_belt_01_l}
    scale=<0.7 0.7 0.7>
    translation {z=-5}
   }
   n69 = "Transform (Modify) Bone - Bone: driver_loincloth_01_l"(
    ComponentPose=Local To Component(LocalPose=Use cached pose 'AnimCache'())
    Alpha=1
    Translation="0,0,0"
    Rotation="0.000,0.000,-10.001"
    Scale="1,1,1"
   ) {rotationMode=BMM_Additive, alphaBoolBlend {blendOption=Linear}, boneToModify {boneName=driver_loincloth_01_l}, rotation {roll=-10.0014}}
   n70 = "Transform (Modify) Bone - Bone: driver_loincloth_01_r"(ComponentPose=@n69, Alpha=1, Translation="0,0,0", Rotation="0.000,0.000,-10.001", Scale="1,1,1") {rotationMode=BMM_Additive, alphaBoolBlend {blendOption=Linear}, boneToModify {boneName=driver_loincloth_01_r}, rotation {roll=-10.0014}}
   n71 = Get cloth_fast_alpha() -> (cloth_fast_alpha:double)
   n72 = "Transform (Modify) Bone - Bone: driver_loincloth_01_l"(ComponentPose=@n70, Alpha=@n71, Translation="0,0,0", Rotation="0.000,0.000,-55", Scale="1,1,1") {rotationMode=BMM_Additive, alphaBoolBlend {blendOption=Linear}, boneToModify {boneName=driver_loincloth_01_l}, rotation {roll=-55}}
   n73 = "Transform (Modify) Bone - Bone: driver_loincloth_01_r"(ComponentPose=@n72, Alpha=@n71, Translation="0,0,0", Rotation="0.000,0.000,-55", Scale="1,1,1") {rotationMode=BMM_Additive, alphaBoolBlend {blendOption=Linear}, boneToModify {boneName=driver_loincloth_01_r}, rotation {roll=-55}}
   n74 = "Transform (Modify) Bone - Bone: driver_dyn_belt_01_r"(ComponentPose=@n73, Alpha=1, Translation="0,0,0", Rotation="5,-0.601,-5", Scale="1,1,1") {
    rotationMode=BMM_Additive
    rotationSpace=BCS_BoneSpace
    translationSpace=BCS_BoneSpace
    alphaBoolBlend {blendOption=Linear}
    boneToModify {boneName=driver_dyn_belt_01_r}
    rotation=<rot 5 -0.601 -5>
   }
   n75 = "Transform (Modify) Bone - Bone: driver_dyn_belt_01_l"(ComponentPose=@n74, Alpha=1, Translation="0,0,0", Rotation="-8,-7.5,-6.5", Scale="1,1,1") {
    rotationMode=BMM_Additive
    rotationSpace=BCS_BoneSpace
    translationSpace=BCS_BoneSpace
    alphaBoolBlend {blendOption=Linear}
    boneToModify {boneName=driver_dyn_belt_01_l}
    rotation=<rot -8 -7.5 -6.5>
   }
   n76 = "Transform (Modify) Bone - Bone: driver_dyn_belt_01_r"(ComponentPose=@n75, Alpha=@n71, Translation="0,0,0", Rotation="60,-0.601,-5", Scale="1,1,1") {
    rotationMode=BMM_Additive
    rotationSpace=BCS_BoneSpace
    translationSpace=BCS_BoneSpace
    alphaBoolBlend {blendOption=Linear}
    boneToModify {boneName=driver_dyn_belt_01_r}
    rotation=<rot 60 -0.601 -5>
   }
   n77 = "Transform (Modify) Bone - Bone: driver_dyn_belt_02_r"(ComponentPose=@n76, Alpha=1, Translation="0,0,0", Rotation="60,-0.601,-5", Scale="1,1,1") {
    rotationMode=BMM_Additive
    rotationSpace=BCS_BoneSpace
    translationSpace=BCS_BoneSpace
    alphaBoolBlend {blendOption=Linear}
    boneToModify {boneName=driver_dyn_belt_02_r}
    rotation=<rot 60 -0.601 -5>
   }
   n78 = RigidBody(ComponentPose=@n77, Alpha=Get belt_off_alpha() -> (belt_off_alpha:double)) {
    bClampLinearTranslationLimitToRefPose=true
    bOverrideWorldGravity=true
    overridePhysicsAsset=&/M_MED_Dragon_Ninja_T1_Physics1
    simulationSpace=BaseBoneSpace
    alphaBoolBlend {blendOption=Linear}
    baseBoneRef {boneName=spine_01}
    componentAppliedLinearAccClamp=<2000 2000 2000>
    componentLinearAccScale=<1 0.7 0.7>
    componentLinearVelScale=<1 0.7 0.7>
    overrideWorldGravity {z=-980}
   }
   n79 = Blend Poses by bool(
    BlendPose_0=Component To Local(ComponentPose=@n68)
    BlendPose_1=Component To Local(ComponentPose=@n78)
    bActiveValue=Get disable_custom_cloth() -> (disable_custom_cloth:bool)
   ) {blendType=Linear}
   n80 = "Constraint - driver_loincloth_01_l (P)"(ComponentPose=Local To Component(LocalPose=Use cached pose 'AnimCache'()), Alpha=1, ConstraintWeights_0=1) {
    alphaBoolBlend {blendOption=Linear}
    boneToModify {boneName=driver_loincloth_01_l}
    constraintSetup [{targetBone {boneName=thigh_l}}]
    constraintWeights [1]
   }
   n81 = "Constraint - driver_loincloth_01_r (P)"(ComponentPose=@n80, Alpha=1, ConstraintWeights_0=1) {
    alphaBoolBlend {blendOption=Linear}
    boneToModify {boneName=driver_loincloth_01_r}
    constraintSetup [{targetBone {boneName=thigh_r}}]
    constraintWeights [1]
   }
   n82 = Anim Dynamics - Bone: dyn_loincloth_02_l(ComponentPose=@n81, Alpha=1, ExternalForce="0,0,0") {
    angularDampingOverride=0.9995
    angularSpringConstant=300
    bAngularSpring=true
    bGravityOverrideInSimSpace=true
    bOverrideAngularDamping=true
    gravityScale=0.33
    simulationSpace=RootRelative
    alphaBoolBlend {blendOption=Linear}
    boundBone {boneName=dyn_loincloth_02_l}
    physicsBodyDefinitions [{
      sphereCollisionRadius=0
      boundBone {boneName=dyn_loincloth_02_l}
      boxExtents=<10 10 10>
      constraintSetup {angularLimitsMax=<1 1 15>, angularLimitsMin=<-25 -20 -15>, angularTarget=<0 0 0>}
      localJointOffset=<1 0 0>
     }]
   }
   n83 = Anim Dynamics - Bone: dyn_loincloth_03A_l(ComponentPose=@n82, Alpha=1, ExternalForce="0,0,0") {
    angularDampingOverride=0.9995
    angularSpringConstant=300
    bAngularSpring=true
    bGravityOverrideInSimSpace=true
    bOverrideAngularDamping=true
    gravityScale=0.33
    simulationSpace=RootRelative
    alphaBoolBlend {blendOption=Linear}
    boundBone {boneName=dyn_loincloth_03A_l}
    physicsBodyDefinitions [{
      sphereCollisionRadius=0
      boundBone {boneName=dyn_loincloth_03A_l}
      boxExtents=<10 10 10>
      constraintSetup {twistAxis=AxisZ, angularLimitsMax=<1 20 15>, angularLimitsMin=<-25 -20 -1>, angularTarget=<0 0 0>}
      localJointOffset=<0 0 -5>
     }]
   }
   n84 = Anim Dynamics - Bone: dyn_loincloth_02_r(ComponentPose=@n83, Alpha=1, ExternalForce="0,0,0") {
    angularDampingOverride=0.9995
    angularSpringConstant=300
    bAngularSpring=true
    bGravityOverrideInSimSpace=true
    bOverrideAngularDamping=true
    gravityScale=0.33
    simulationSpace=RootRelative
    alphaBoolBlend {blendOption=Linear}
    boundBone {boneName=dyn_loincloth_02_r}
    physicsBodyDefinitions [{
      sphereCollisionRadius=0
      boundBone {boneName=dyn_loincloth_02_r}
      boxExtents=<10 10 10>
      constraintSetup {angularLimitsMax=<25 1 15>, angularLimitsMin=<1 -20 -15>, angularTarget=<0 0 0>}
      localJointOffset=<-1 0 0>
     }]
   }
   n85 = Anim Dynamics - Bone: dyn_loincloth_03A_r(ComponentPose=@n84, Alpha=1, ExternalForce="0,0,0") {
    angularDampingOverride=0.9995
    angularSpringConstant=300
    bAngularSpring=true
    bGravityOverrideInSimSpace=true
    bOverrideAngularDamping=true
    gravityScale=0.33
    simulationSpace=RootRelative
    alphaBoolBlend {blendOption=Linear}
    boundBone {boneName=dyn_loincloth_03A_r}
    physicsBodyDefinitions [{
      sphereCollisionRadius=0
      boundBone {boneName=dyn_loincloth_03A_r}
      boxExtents=<10 10 10>
      constraintSetup {twistAxis=AxisZ, angularLimitsMax=<1 20 15>, angularLimitsMin=<-25 -20 -1>, angularTarget=<0 0 0>}
      localJointOffset=<0 0 -5>
     }]
   }
   n86 = "Transform (Modify) Bone - Bone: driver_dyn_belt_01_r"(ComponentPose=@n85, Alpha=1, Translation="-1,0,0", Rotation="-28,0,-1", Scale="1,1,1") {
    rotationMode=BMM_Additive
    rotationSpace=BCS_BoneSpace
    translationMode=BMM_Additive
    translationSpace=BCS_BoneSpace
    alphaBoolBlend {blendOption=Linear}
    boneToModify {boneName=driver_dyn_belt_01_r}
    rotation {pitch=-28, roll=-1}
    translation {x=-1}
   }
   n87 = "Constraint - parent_dyn_belt_01_l (P)"(ComponentPose=@n86, Alpha=1, ConstraintWeights_0=1) {
    alphaBoolBlend {blendOption=Linear}
    boneToModify {boneName=parent_dyn_belt_01_l}
    constraintSetup [{targetBone {boneName=dyn_loincloth_01_l}}]
    constraintWeights [1]
   }
   n88 = "Transform (Modify) Bone - Bone: driver_dyn_belt_01_l"(ComponentPose=@n87, Alpha=1, Translation="0,0,0", Rotation="0,0,0", Scale="0.7,0.7,0.7") {
    rotationSpace=BCS_BoneSpace
    scaleMode=BMM_Additive
    scaleSpace=BCS_BoneSpace
    translationSpace=BCS_BoneSpace
    alphaBoolBlend {blendOption=Linear}
    boneToModify {boneName=driver_dyn_belt_01_l}
    scale=<0.7 0.7 0.7>
   }
   n89 = "Transform (Modify) Bone - Bone: Flap_Front_Driven_Bone"(ComponentPose=Local To Component(LocalPose=Use cached pose 'AnimCache'()), Alpha=1, Translation="0,0,7", Rotation="0,0,-40", Scale="1,1,1") {
    rotationMode=BMM_Additive
    rotationSpace=BCS_BoneSpace
    translationMode=BMM_Additive
    translationSpace=BCS_BoneSpace
    alphaBoolBlend {blendOption=Linear}
    boneToModify {boneName=Flap_Front_Driven_Bone}
    rotation {roll=-40}
    translation {z=7}
   }
   n90 = Flap Node - Bone: dyn_Flap_Front!(ComponentPose:Component Space Pose Link Structure=@n89, Alpha:float=1) -> (Pose:Component Space Pose Link Structure) {
    is=title FortAnimGraphNode_Flap
    bUseClamp=true
    targetClampMax=100
    alphaBoolBlend {blendOption=Linear}
    sourceDrives [
     {
      bUseRange=false
      drivingCurve=null
      multiplier=0.5
      rangeMax=1
      rangeMin=-1
      remappedMax=1
      remappedMin=0
      sourceComponent=RotationZ
      useQuaternion=true
      sourceBone {boneName=thigh_l}
     }
     {
      bUseRange=false
      drivingCurve=null
      multiplier=0.5
      rangeMax=1
      rangeMin=-1
      remappedMax=1
      remappedMin=0
      sourceComponent=RotationZ
      useQuaternion=true
      sourceBone {boneName=thigh_r}
     }
    ]
    targetBone {boneName=dyn_Flap_Front}
   }
   n91 = "dyn_Flap_Front.rotateZ = thigh_l.rotateZ * -0.5 - Bone Driven Controller"!(ComponentPose:Component Space Pose Link Structure=@n90, Alpha:float=0.95) -> (Pose:Component Space Pose Link Structure) {
    is=title AnimGraphNode_BoneDrivenController
    bAffectTargetRotationZ=true
    multiplier=-0.5
    sourceComponent=RotationZ
    alphaBoolBlend {blendOption=Linear}
    sourceBone {boneName=thigh_l}
    targetBone {boneName=dyn_Flap_Front}
   }
   n92 = "dyn_Flap_Front.rotateZ = thigh_r.rotateZ * 0.5 - Bone Driven Controller"!(ComponentPose:Component Space Pose Link Structure=@n91, Alpha:float=0.95) -> (Pose:Component Space Pose Link Structure) {
    is=title AnimGraphNode_BoneDrivenController
    bAffectTargetRotationZ=true
    multiplier=0.5
    sourceComponent=RotationZ
    alphaBoolBlend {blendOption=Linear}
    sourceBone {boneName=thigh_r}
    targetBone {boneName=dyn_Flap_Front}
   }
   n93 = "Constraint - driver_loincloth_01_l (P)"(ComponentPose=@n92, Alpha=0.35, ConstraintWeights_0=1) {
    alphaBoolBlend {blendOption=Linear}
    boneToModify {boneName=driver_loincloth_01_l}
    constraintSetup [{targetBone {boneName=thigh_l}}]
    constraintWeights [1]
   }
   n94 = "Constraint - driver_loincloth_01_r (P)"(ComponentPose=@n93, Alpha=0.7, ConstraintWeights_0=1) {
    alphaBoolBlend {blendOption=Linear}
    boneToModify {boneName=driver_loincloth_01_r}
    constraintSetup [{targetBone {boneName=thigh_r}}]
    constraintWeights [1]
   }
   n95 = "Constraint - parent_dyn_belt_01_l (P)"(ComponentPose=@n94, Alpha=1, ConstraintWeights_0=1) {
    alphaBoolBlend {blendOption=Linear}
    boneToModify {boneName=parent_dyn_belt_01_l}
    constraintSetup [{targetBone {boneName=dyn_loincloth_01_l}}]
    constraintWeights [1]
   }
   n96 = Blend Poses by bool(
    BlendPose_0=Component To Local(ComponentPose=@n95)
    BlendPose_1=Component To Local(ComponentPose=@n88)
    bActiveValue=Get crouch_moving() -> (crouch_moving:bool)
   ) {blendType=Linear}
   n97 = Get is_crouching() -> (is_crouching:bool)
   n98 = "Constraint - parent_dyn_belt_01_r (P)"(
    ComponentPose=Local To Component(LocalPose=Blend Poses by bool(BlendPose_0=@n96, BlendPose_1=@n79, bActiveValue=@n97) {blendType=Linear})
    Alpha=1
    ConstraintWeights_0=1
   ) {
    alphaBoolBlend {blendOption=Linear}
    boneToModify {boneName=parent_dyn_belt_01_r}
    constraintSetup [{targetBone {boneName=dyn_loincloth_01_r}}]
    constraintWeights [1]
   }
   n99 = Anim Dynamics - Bone: dyn_belt_01_r(ComponentPose=@n98, Alpha=1, ExternalForce="0,0,0") {
    angularDampingOverride=0.9995
    bGravityOverrideInSimSpace=true
    bOverrideAngularDamping=true
    gravityScale=0.33
    simulationSpace=RootRelative
    alphaBoolBlend {blendOption=Linear}
    boundBone {boneName=dyn_belt_01_r}
    physicsBodyDefinitions [{
      sphereCollisionRadius=0
      boundBone {boneName=dyn_belt_01_r}
      boxExtents=<20 20 20>
      constraintSetup {twistAxis=AxisZ, angularLimitsMax=<25 25 -1>, angularLimitsMin=<-1 1 -1>, angularTarget=<0 0 0>}
      localJointOffset=<0 0 -5>
     }]
   }
   n100 = Look At - Bone: driver_dyn_belt_02_r(ComponentPose=@n99, Alpha=1) {
    alphaBoolBlend {blendOption=Linear}
    boneToModify {boneName=driver_dyn_belt_02_r}
    lookAtLocation=<0 5 -10>
    lookAtTarget {boneReference {boneName=dyn_loincloth_03A_r}}
    lookAt_Axis {axis {y=0, z=-1}}
   }
   n101 = Anim Dynamics - Bone: dyn_belt_02_r(ComponentPose=@n100, Alpha=1, ExternalForce="0,0,0") {
    angularDampingOverride=0.9995
    bGravityOverrideInSimSpace=true
    bOverrideAngularDamping=true
    gravityScale=0.33
    simulationSpace=RootRelative
    alphaBoolBlend {blendOption=Linear}
    boundBone {boneName=dyn_belt_02_r}
    physicsBodyDefinitions [{
      sphereCollisionRadius=0
      boundBone {boneName=dyn_belt_02_r}
      boxExtents=<20 20 20>
      constraintSetup {twistAxis=AxisZ, angularLimitsMax=<25 25 5>, angularLimitsMin=<-1 -5 -5>, angularTarget=<0 0 0>}
      localJointOffset=<0 0 -5>
     }]
   }
   n102 = Anim Dynamics - Bone: dyn_belt_03_r(ComponentPose=@n101, Alpha=1, ExternalForce="0,0,0") {
    angularDampingOverride=0.9995
    bGravityOverrideInSimSpace=true
    bOverrideAngularDamping=true
    gravityScale=0.33
    simulationSpace=RootRelative
    alphaBoolBlend {blendOption=Linear}
    boundBone {boneName=dyn_belt_03_r}
    physicsBodyDefinitions [{
      sphereCollisionRadius=0
      boundBone {boneName=dyn_belt_03_r}
      boxExtents=<20 20 20>
      constraintSetup {twistAxis=AxisZ, angularLimitsMax=<25 25 5>, angularLimitsMin=<-1 -5 -5>, angularTarget=<0 0 0>}
      localJointOffset=<0 0 -5>
     }]
   }
   n103 = Anim Dynamics - Bone: dyn_belt_04_r(ComponentPose=@n102, Alpha=1, ExternalForce="0,0,0") {
    angularDampingOverride=0.9995
    bGravityOverrideInSimSpace=true
    bOverrideAngularDamping=true
    gravityScale=0.33
    simulationSpace=RootRelative
    alphaBoolBlend {blendOption=Linear}
    boundBone {boneName=dyn_belt_04_r}
    physicsBodyDefinitions [{
      sphereCollisionRadius=0
      boundBone {boneName=dyn_belt_04_r}
      boxExtents=<20 20 20>
      constraintSetup {twistAxis=AxisZ, angularLimitsMax=<25 25 5>, angularLimitsMin=<-1 -5 -5>, angularTarget=<0 0 0>}
      localJointOffset=<0 0 -5>
     }]
   }
   n104 = Get crouch_off_alpha() -> (crouch_off_alpha:double)
   n105 = "Constraint - parent_dyn_belt_01_l (R)"(ComponentPose=@n103, Alpha=@n104, ConstraintWeights_0=1) {
    alphaBoolBlend {blendOption=Linear}
    boneToModify {boneName=parent_dyn_belt_01_l}
    constraintSetup [{transformType=Rotation, targetBone {boneName=dyn_loincloth_01_l}}]
    constraintWeights [1]
   }
   n106 = Anim Dynamics - Bone: dyn_belt_01_l(ComponentPose=@n105, Alpha=@n104, ExternalForce="0,0,0") {
    angularDampingOverride=0.9995
    bGravityOverrideInSimSpace=true
    bOverrideAngularDamping=true
    gravityScale=0.33
    simulationSpace=RootRelative
    alphaBoolBlend {blendOption=Linear}
    boundBone {boneName=dyn_belt_01_l}
    physicsBodyDefinitions [{
      sphereCollisionRadius=0
      boundBone {boneName=dyn_belt_01_l}
      boxExtents=<20 20 20>
      constraintSetup {twistAxis=AxisZ, angularLimitsMax=<25 15 -1>, angularLimitsMin=<-1 -20 -1>, angularTarget=<0 0 0>}
      localJointOffset=<0 0 -5>
     }]
   }
   n107 = Look At - Bone: driver_dyn_belt_02_l(ComponentPose=@n106, Alpha=@n104) {
    alphaBoolBlend {blendOption=Linear}
    boneToModify {boneName=driver_dyn_belt_02_l}
    lookAtLocation=<0 5 -10>
    lookAtTarget {boneReference {boneName=dyn_loincloth_03A_l}}
    lookAt_Axis {axis {y=0, z=-1}}
   }
   n108 = Anim Dynamics - Bone: dyn_belt_02_l(ComponentPose=@n107, Alpha=@n104, ExternalForce="0,0,0") {
    angularDampingOverride=0.9995
    bGravityOverrideInSimSpace=true
    bOverrideAngularDamping=true
    gravityScale=0.33
    simulationSpace=RootRelative
    alphaBoolBlend {blendOption=Linear}
    boundBone {boneName=dyn_belt_02_l}
    physicsBodyDefinitions [{
      sphereCollisionRadius=0
      boundBone {boneName=dyn_belt_02_l}
      boxExtents=<20 20 20>
      constraintSetup {twistAxis=AxisZ, angularLimitsMax=<25 15 5>, angularLimitsMin=<-1 -20 -5>, angularTarget=<0 0 0>}
      localJointOffset=<0 0 -5>
     }]
   }
   n109 = Anim Dynamics - Bone: dyn_belt_03_l(ComponentPose=@n108, Alpha=@n104, ExternalForce="0,0,0") {
    angularDampingOverride=0.9995
    bGravityOverrideInSimSpace=true
    bOverrideAngularDamping=true
    gravityScale=0.33
    simulationSpace=RootRelative
    alphaBoolBlend {blendOption=Linear}
    boundBone {boneName=dyn_belt_03_l}
    physicsBodyDefinitions [{
      sphereCollisionRadius=0
      boundBone {boneName=dyn_belt_03_l}
      boxExtents=<20 20 20>
      constraintSetup {twistAxis=AxisZ, angularLimitsMax=<25 15 5>, angularLimitsMin=<-1 -20 -5>, angularTarget=<0 0 0>}
      localJointOffset=<0 0 -5>
     }]
   }
   n110 = Anim Dynamics - Bone: dyn_belt_04_l(ComponentPose=@n109, Alpha=@n104, ExternalForce="0,0,0") {
    angularDampingOverride=0.9995
    bGravityOverrideInSimSpace=true
    bOverrideAngularDamping=true
    gravityScale=0.33
    simulationSpace=RootRelative
    alphaBoolBlend {blendOption=Linear}
    boundBone {boneName=dyn_belt_04_l}
    physicsBodyDefinitions [{
      sphereCollisionRadius=0
      boundBone {boneName=dyn_belt_04_l}
      boxExtents=<20 20 20>
      constraintSetup {twistAxis=AxisZ, angularLimitsMax=<25 15 5>, angularLimitsMin=<-1 -20 -5>, angularTarget=<0 0 0>}
      localJointOffset=<0 0 -5>
     }]
   }
   n111 = RigidBody(ComponentPose=@n110, Alpha=0.95) {
    bOverrideWorldGravity=true
    overridePhysicsAsset=&/M_MED_Dragon_Ninja_T1_ClothRear_Physics
    simulationSpace=BaseBoneSpace
    alphaBoolBlend {blendOption=Linear}
    baseBoneRef {boneName=spine_01}
    componentAppliedLinearAccClamp=<2000 2000 2000>
    componentLinearAccScale=<1 0.7 0.7>
    componentLinearVelScale=<1 0.7 0.7>
    overrideWorldGravity {z=-980}
   }
   n112 = "Transform (Modify) Bone - Bone: cape_attach"(
    ComponentPose=@n111
    Alpha=Get spine_rot_alpha() -> (spine_rot_alpha:double)
    Translation=Get cape_trans() -> (cape_trans:Vector)
    Rotation=Get cape_rotator() -> (cape_rotator:Rotator)
    Scale="1,1,1"
   ) {
    rotationMode=BMM_Additive
    rotationSpace=BCS_BoneSpace
    translationMode=BMM_Additive
    translationSpace=BCS_BoneSpace
    alphaBoolBlend {blendOption=Linear}
    boneToModify {boneName=cape_attach}
   }
   Save cached pose 'PreSkyDiving'(Pose=Component To Local(ComponentPose=@n112)) {cacheName=PreSkyDiving}
   Sequence Player 'dninjaarms'() {playRate=0, sequence=&/dninjaarms, startPosition=0.5}
   Get ads_tip_alpha()
  ]}
 EventGraph {nodes [
   n1 = Event BlueprintUpdateAnimation()
   > Get!(execute:Exec) -> (then:Exec, else:Exec, Skeletal_Mesh:Skeletal Mesh Component Object Reference)
   > n2 = Sequence!(execute) -> (then_0, then_1, then_2:Exec, then_3:Exec, then_4:Exec, then_5:Exec, then_6:Exec)
   n3 = GetOwnerSkeletalMeshComponent()
   n4 = BreakRotator(InRot=BreakTransform(InTransform=GetDeltaTransformFromRefPose(self=@n3, BoneName=upperarm_r, BaseName=clavicle_r)).Rotation)
   Set arm_roll_back_r(
    execute=@n2.then_1
    arm_roll_back_r:double=FInterpTo(Current=Get arm_roll_back_r(), Target=MapRangeClamped(Value=@n4, InRangeA=0, InRangeB=-25), DeltaTime=@n1.DeltaTimeX, InterpSpeed=20)
   ) -> (Output_Get:double)
   > Set arm_roll_fwd_r(arm_roll_fwd_r:double=FInterpTo(Current=Get arm_roll_fwd_r(), Target=MapRangeClamped(Value=@n4, InRangeA=0, InRangeB=35), DeltaTime=@n1.DeltaTimeX, InterpSpeed=20)) -> (Output_Get:double)
   > Set arm_up_r(arm_up_r:double=MapRangeClamped(Value=@n4.Yaw, InRangeA=0, InRangeB=130, OutRangeB=0.75)) -> (Output_Get:double)
   n5 = GetDeltaTransformFromRefPose(self=@n3, BoneName=clavicle_r, BaseName=spine_05)
   > Set R Arm Pitch SDFix(R Arm Pitch SDFix:double=BreakRotator(InRot=BreakTransform(InTransform=@n5).Rotation).Yaw) -> (Output_Get:double)
   > Set arm_pitch_up_r(arm_pitch_up_r:double=MapRangeClamped(Value=@n4.Pitch, InRangeA=0, InRangeB=70)) -> (Output_Get:double)
   n6 = BreakRotator(InRot=BreakTransform(InTransform=GetDeltaTransformFromRefPose(self=@n3, BoneName=upperarm_r, BaseName=spine_05)).Rotation)
   > Set arm_yaw_up_r(arm_yaw_up_r:double=MapRangeClamped(Value=@n6.Yaw, InRangeA=0, InRangeB=70)) -> (Output_Get:double)
   n7 = "float > float"(A=@n6.Yaw, B=110) {is=title K2Node_CallFunction}
   n8 = "float < float"!(A:double=@n6.Yaw, B:double=-100) -> (ReturnValue:bool) {is=title K2Node_CallFunction}
   > Branch(Condition=AND Boolean(A="float > float"(A=@n6.Pitch, B=50), B=OR Boolean(A=@n7, B=@n8)))
   n9 = Get arm_roll_back_r()
   > Set arm_roll_back_r(arm_roll_back_r=FInterpTo(Current=@n9, Target=0, DeltaTime=@n1.DeltaTimeX, InterpSpeed=30))
   n10 = Get arm_roll_fwd_r()
   > Set arm_roll_fwd_r(arm_roll_fwd_r=FInterpTo(Current=@n10, Target=0, DeltaTime=@n1.DeltaTimeX, InterpSpeed=30))
   Reroute Node(InputPin=@n2.then_5)
   n11 = Branch(execute=@n2.then_0, Condition=Get bIsCrouching())
   n12 = Get PawnSpeed2D()
   > n13 = Branch(Condition="float <= float"!(A:double=MapRangeClamped(Value=@n12, InRangeA=0, InRangeB=150), B:double=0) -> (ReturnValue:bool) {is=title K2Node_CallFunction})
   > Set crouch_speed_alpha(crouch_speed_alpha:double=1) -> (Output_Get:double)
   > Set crouch_moving(crouch_moving:bool=false) -> (Output_Get:bool)
   > Set crouch_off_alpha(crouch_off_alpha:double=0) -> (Output_Get:double)
   Set crouch_speed_alpha(execute=@n13.else, crouch_speed_alpha=0)
   > Set crouch_moving(crouch_moving=true)
   > Set crouch_off_alpha(crouch_off_alpha=1)
   Set crouch_off_alpha(execute=@n11.else, crouch_off_alpha=1)
   > Set belt_off_alpha(belt_off_alpha:double=MapRangeClamped(Value=@n12, InRangeA=400, InRangeB=540, OutRangeA=1, OutRangeB=0.4)) -> (Output_Get:double)
   > Set cloth_fast_alpha(cloth_fast_alpha:double=MapRangeClamped(Value=@n12, InRangeA=400, InRangeB=540)) -> (Output_Get:double)
   Branch(execute=@n2.then_3, Condition=Get bIsSkyDiving())
   > n14 = Branch(Condition=Get bIsParachuteOpened())
   > n15 = Branch(Condition=IsUsingUmbrella!(self:Fort Player Pawn Object Reference=Get Pawn() -> (Pawn:Fort Player Pawn Object Reference)) -> (ReturnValue:bool))
   > Set skydiving_umbrella_active(skydiving_umbrella_active:bool=true) -> (Output_Get:bool)
   > Set skydive_falling_alpha(skydive_falling_alpha:double=0) -> (Output_Get:double)
   > Set skydive_parachute_alpha(skydive_parachute_alpha:double=0) -> (Output_Get:double)
   > Set skydive_umbrella_alpha(skydive_umbrella_alpha:double=1) -> (Output_Get:double)
   Set skydiving_umbrella_active(execute=@n15.else, skydiving_umbrella_active=false)
   > Set skydive_falling_alpha(skydive_falling_alpha=0)
   n16 = FInterpTo(Current=Get skydive_parachute_alpha(), Target=1, DeltaTime=@n1.DeltaTimeX, InterpSpeed=2)
   > Set skydive_parachute_alpha(skydive_parachute_alpha=@n16)
   > n17 = Branch(Condition="float <= float"(A=Get R Arm Pitch SDFix() -> (R Arm Pitch SDFix:double), B=5))
   > Set skydive_parachute_alpha(skydive_parachute_alpha=float * float(A=@n16, B=0.87))
   Set skydive_parachute_alpha(execute=@n17.else, skydive_parachute_alpha=@n16)
   Set skydive_parachute_alpha(execute=@n14.else, skydive_parachute_alpha=0)
   > n18 = Branch(Condition=Get bIsSkydiveFloating())
   > Set skydive_falling_alpha(skydive_falling_alpha=1)
   Set skydive_falling_alpha(execute=@n18.else, skydive_falling_alpha=0)
   n19 = Get AnimInst() -> (AnimInst:Anim Instance Object Reference)
   Is Valid(exec=@n2.then_4, InputObject=@n19)
   > Set FrontEnd_Idle_On(FrontEnd_Idle_On:bool="float > float"(A=GetCurveValue(self=@n19, CurveName=FrontEnd_Idle_Bypass), B=0)) -> (Output_Get:bool)
   > Set disable_custom_cloth(disable_custom_cloth:bool="float > float"(A=GetCurveValue(self=@n19, CurveName=Disable_Custom), B=0)) -> (Output_Get:bool)
   > Branch(Condition=Get disable_custom_cloth())
   > Set crouch_off_alpha(crouch_off_alpha=0)
   n20 = GetOwnerSkeletalMeshComponent()
   n21 = BreakRotator(InRot=BreakTransform(InTransform=GetDeltaTransformFromRefPose(self=@n20, BoneName=upperarm_l, BaseName=clavicle_l)).Rotation)
   Set arm_roll_back_l(
    execute=@n2.then_2
    arm_roll_back_l:double=FInterpTo(Current=Get arm_roll_back_l(), Target=MapRangeClamped(Value=@n21, InRangeA=0, InRangeB=-20), DeltaTime=@n1.DeltaTimeX, InterpSpeed=25)
   ) -> (Output_Get:double)
   > Set arm_roll_fwd_l(arm_roll_fwd_l:double=FInterpTo(Current=Get arm_roll_fwd_l(), Target=MapRangeClamped(Value=@n21, InRangeA=60, InRangeB=120), DeltaTime=@n1.DeltaTimeX, InterpSpeed=25)) -> (Output_Get:double)
   > Set arm_pitch_up_l(arm_pitch_up_l:double=MapRangeClamped(Value=@n21.Pitch, InRangeA=60, InRangeB=116)) -> (Output_Get:double)
   n22 = Branch(execute=@n2.then_6, Condition=Get bIsInFrontEnd())
   Set spine_rot_alpha(execute=[@n22.else, @n22.then], spine_rot_alpha:double=0) -> (Output_Get:double)
   > Set cape_trans(cape_trans:Vector="0, 0, 0") -> (Output_Get:Vector)
   > Set cape_rotator(cape_rotator:Rotator="0, 0, 0") -> (Output_Get:Rotator)
   > n23 = Branch(Condition=Get bIsSkyDiving())
   > Set spine_rot_alpha(spine_rot_alpha=0)
   > Set cape_rotator(cape_rotator="0, 0,20")
   Branch(
    execute=@n23.else
    Condition=Montage_IsActive!(
     self:Anim Instance Object Reference=Get AnimInst()
     Montage:Anim Montage Object Reference=/Game/Animation/Game/MainPlayer/Menu/BR/CustomSkins/DragonNinja/DragonNinja_T3_Male_Idle_Montage
    ) -> (ReturnValue:bool)
   )
   n24 = Get CurrentLOD()
   n25 = Get bClothEnabled()
   > Branch(Condition=OR Boolean(A=NOT Boolean(A=@n25), B="integer >= integer"!(A:int=@n24, B:int=1) -> (ReturnValue:bool) {is=title K2Node_CallFunction}))
   > Set spine_rot_alpha(spine_rot_alpha=1)
   > Set cape_trans(cape_trans="0,5.3,-5")
   > Set cape_rotator(cape_rotator="0, 0,40")
   Event BlueprintInitializeAnimation()
   n26 = GetOwnerSkeletalMeshComponent()
   > Is Valid(InputObject=@n26)
   > Set Skeletal_Mesh(Skeletal_Mesh:Skeletal Mesh Component Object Reference=@n26) -> (Output_Get:Skeletal Mesh Component Object Reference)
   > n27 = Cast To FortPlayerPawn(Object=GetOwningActor())
   > Is Valid(InputObject=@n27)
   > n28 = Set Pawn(Pawn:Fort Player Pawn Object Reference=@n27) -> (Output_Get:Fort Player Pawn Object Reference)
   n29 = GetAnimInstance(self=Get Mesh(self=@n28))
   > Is Valid(InputObject=@n29)
   > Set AnimInst(AnimInst:Anim Instance Object Reference=@n29) -> (Output_Get:Anim Instance Object Reference)
   Event OnBeginTargeting() {is=title K2Node_Event}
   > Set ads_alpha(ads_alpha:double=1) -> (Output_Get:double)
   > Set ads_tip_alpha(ads_tip_alpha:double=0.25) -> (Output_Get:double)
   Event OnEndTargeting() {is=title K2Node_Event}
   > Set ads_alpha(ads_alpha=0)
   > Set ads_tip_alpha(ads_tip_alpha=0.85)
   Event OnBeginCrouching()
   > Set is_crouching(is_crouching:bool=true) -> (Output_Get:bool)
   > Set crouch_off_alpha(crouch_off_alpha=0)
   Event OnEndCrouching()
   > Set is_crouching(is_crouching=false)
   > Set crouch_off_alpha(crouch_off_alpha=1)
   Event OnCharacterCustomizationLoaded()
   > Do Once()
   > n30 = Set isWearingBackbling(isWearingBackbling=CheckForBackbling())
   > Branch(Condition=@n30)
   > SetMorphTarget!(execute, self:Anim Instance Object Reference, MorphTargetName:Name=hide_scales, Value:float=1) -> (then)
   > Set hide_fins(hide_fins:bool=true) -> (Output_Get:bool)
   n31 = "Absolute (Float)"(A=GetOwnerDirection!(self:Custom Character Part Anim Instance Object Reference) -> (ReturnValue:float))
   n32 = Branch(Condition="float < float"(A=@n31, B=100))
   > Branch(Condition="float > float"(A=@n31, B=30))
   Branch(execute=@n32.else, Condition="float > float"(A=@n31, B=100))
   n33 = Branch(Condition=Get is_crouching())
   n34 = Reroute Node(InputPin:Skinned Mesh Component Object Reference) -> (OutputPin:Skinned Mesh Component Object Reference)
   n35 = BreakRotator(InRot=BreakTransform(InTransform=GetDeltaTransformFromRefPose(self=@n34, BoneName=upperarm_twist_01_r)).Rotation)
   n36 = MapRangeClamped(Value=@n35, InRangeA=0, InRangeB=40)
   > Set arm_roll_inwards_r(arm_roll_inwards_r:double=float * float(A=@n36, B=0.5)) -> (Output_Get:double)
   Set arm_roll_inwards_r(execute=@n33.else, arm_roll_inwards_r=@n36)
   n37 = BreakRotator(InRot=BreakTransform(InTransform=GetDeltaTransformFromRefPose(self=@n34, BoneName=upperarm_twist_01_r, BaseName=spine_05)).Rotation)
   > Set arm_up_r(arm_up_r=MapRangeClamped(Value=@n37.Yaw, InRangeA=20, InRangeB=35))
   > n38 = Branch(Condition=AND Boolean(A="float >= float"(A=@n37.Yaw, B=20) {is=title K2Node_CallFunction}, B="float <= float"(A=@n37, B=10)))
   > n39 = Set arm_ready(arm_ready:bool=true) -> (Output_Get:bool)
   n40 = Branch(execute=[@n41.then, @n39.then], Condition=Get arm_ready() -> (arm_ready:bool))
   > Set arm_roll_fwd_back_r(arm_roll_fwd_back_r:double=FInterpTo(Current=Get arm_roll_fwd_back_r() -> (arm_roll_fwd_back_r:double), Target=1, DeltaTime=0, InterpSpeed=0.5)) -> (Output_Get:double)
   Set arm_roll_fwd_back_r(execute=@n40.else, arm_roll_fwd_back_r=FInterpTo(Current=Get arm_roll_fwd_back_r(), Target=0, DeltaTime=0, InterpSpeed=0.5))
   n41 = Set arm_ready(execute=@n38.else, arm_ready=false)
   n42 = Branch(Condition=Get bIsCrouching())
   > n43 = Set arm_up_r(arm_up_r=0)
   n44 = BreakRotator(InRot=BreakTransform(InTransform=GetDeltaTransformFromRefPose(BoneName=upperarm_twist_01_r, BaseName=spine_05)).Rotation)
   Set arm_up_more_r(execute=[@n49.then, @n43.then], arm_up_more_r:double=MapRangeClamped(Value=@n44.Yaw, InRangeA=85, InRangeB=125)) -> (Output_Get:double)
   > Set arm_back_r(arm_back_r:double=MapRangeClamped(Value=@n44.Yaw, InRangeA=35, InRangeB=-20)) -> (Output_Get:double)
   > Set pad_bck_out(pad_bck_out:double=MapRangeClamped(Value=@n44.Yaw, InRangeA=40, InRangeB=120)) -> (Output_Get:double)
   n45 = Get arm_up_r()
   > n46 = Branch(Condition="float > float"(A=@n45, B=0.5))
   n47 = BreakRotator(InRot=BreakTransform(InTransform=GetDeltaTransformFromRefPose(BoneName=upperarm_r, BaseName=clavicle_r)).Rotation)
   n48 = MapRangeClamped(
    Value=float + float(A=MapRangeClamped(Value=@n47.Yaw, InRangeA=14, InRangeB=45, OutRangeB=0.7), B=MapRangeClamped(Value=@n47, InRangeA=-5, InRangeB=-20)) {is=title K2Node_CommutativeAssociativeBinaryOperator}
    InRangeA=0
    InRangeB=1
   )
   > Set r_pad_roll_fwd(r_pad_roll_fwd:double=float * float(A=@n48, B=float * float(A=@n45, B=0.05))) -> (Output_Get:double)
   Set r_pad_roll_fwd(execute=@n46.else, r_pad_roll_fwd=@n48)
   n49 = Set arm_up_r(execute=@n42.else, arm_up_r=MapRangeClamped(Value=@n44.Yaw, InRangeA=40, InRangeB=80))
   n50 = GetPartSkeletalMeshComponent()
   Set pad_top_r_dist(pad_top_r_dist:double=Get Distance Between Two Sockets!(
     Component:Skeletal Mesh Component Object Reference=@n50
     SocketOrBoneNameA:Name=dyn_pad_upper_pos_r
     SocketSpaceA:ERelativeTransformSpace Enum=RTS_Component
     SocketOrBoneNameB:Name=spine_05
     SocketSpaceB:ERelativeTransformSpace Enum=RTS_Component
     bRemapRange:bool=true
     InRangeMin:float=21
     InRangeMax:float=15
     OutRangeMin:float=0
     OutRangeMax:float=1
    ) -> (ReturnValue:float) {is=title K2Node_CallFunction}) -> (Output_Get:double)
   > Set pad_back_r_dist(pad_back_r_dist:double=Get Distance Between Two Sockets(
     Component=@n50
     SocketOrBoneNameA=dyn_pad_back_pos_r
     SocketSpaceA=RTS_Component
     SocketOrBoneNameB=spine_05
     SocketSpaceB=RTS_Component
     bRemapRange=false
     InRangeMin=21
     InRangeMax=12
     OutRangeMin=0
     OutRangeMax=1
    )) -> (Output_Get:double)
   > Set pad_front_r_dist(pad_front_r_dist:double=Get Distance Between Two Sockets(
     Component=@n50
     SocketOrBoneNameA=dyn_pad_front_pos_r
     SocketSpaceA=RTS_Component
     SocketOrBoneNameB=spine_05
     SocketSpaceB=RTS_Component
     bRemapRange=true
     InRangeMin=21
     InRangeMax=12
     OutRangeMin=0
     OutRangeMax=1
    )) -> (Output_Get:double)
   Set arm_roll_back_l(arm_roll_back_l=FInterpTo(Current=Get arm_roll_back_l(), Target=0, DeltaTime=@n1.DeltaTimeX, InterpSpeed=30))
   > Set arm_roll_fwd_l(arm_roll_fwd_l=FInterpTo(Current=Get arm_roll_fwd_l(), Target=0, DeltaTime=@n1.DeltaTimeX, InterpSpeed=30))
   n51 = BreakRotator(InRot=BreakTransform(InTransform=GetDeltaTransformFromRefPose(BoneName=upperarm_twist_01_r, BaseName=clavicle_r)).Rotation)
   MapRangeClamped(Value=@n51, InRangeA=0, InRangeB=-25)
   MapRangeClamped(Value=@n51.Pitch, InRangeA=0, InRangeB=70)
   MapRangeClamped(Value=@n51, InRangeA=0, InRangeB=35)
   n52 = GetPartSkeletalMeshComponent()
   OR Boolean(
    A="Not Equal (Object)"!(
     A:Object Reference=FindChildCommponentFromAttachSocket!(
      self:Custom Character Part Anim Instance Object Reference
      Parent:Skeletal Mesh Component Object Reference=@n52
      bIncludeAllDescendants:bool=false
      AttachSocket:Name=Cape
     ) -> (ReturnValue:Skeletal Mesh Component Object Reference)
     B:Object Reference
    ) -> (ReturnValue:bool)
    B="Not Equal (Object)"(A=FindChildCommponentFromAttachSocket(Parent=@n52, bIncludeAllDescendants=false, AttachSocket=backpack_br))
   )
   float * float(A=@n10, B=0)
   float * float(A=@n9, B=0)
   Branch()
   Set arm_roll_back_r(arm_roll_back_r=MapRangeClamped(Value=@n35, InRangeA=0, InRangeB=-20))
   Set arm_up_more_r(arm_up_more_r=0)
  ]}
]
variables [
 ads_alpha {type=number}
 ads_tip_alpha {type=number, default=0.85}
 AnimInst {type=AnimInstance, default=null}
 arm_back_r {type=number}
 arm_pitch_up_l {type=number}
 arm_pitch_up_r {type=number}
 arm_ready {type=boolean}
 arm_roll_back_l {type=number}
 arm_roll_back_r {type=number}
 arm_roll_fwd_back_r {type=number}
 arm_roll_fwd_l {type=number}
 arm_roll_fwd_r {type=number}
 arm_roll_inwards_r {type=number}
 arm_up_more_r {type=number}
 arm_up_r {type=number}
 arm_yaw_up_r {type=number}
 belt_off_alpha {type=number}
 cape_rotator {type=Rotator}
 cape_trans {type=Vector}
 cloth_fast_alpha {type=number}
 crouch_moving {type=boolean}
 crouch_off_alpha {type=number}
 crouch_speed_alpha {type=number}
 disable_custom_cloth {type=boolean}
 FrontEnd_Idle_On {type=boolean}
 hide_fins {type=boolean}
 is_crouching {type=boolean}
 isWearingBackbling {type=boolean}
 pad_back_r_dist {type=number}
 pad_bck_out {type=number}
 pad_front_r_dist {type=number}
 pad_top_r_dist {type=number}
 Pawn {type=FortPlayerPawn}
 R Arm Pitch SDFix {type=number}
 r_pad_roll_fwd {type=number}
 Skeletal_Mesh {type=SkeletalMeshComponent}
 skydive_falling_alpha {type=number}
 skydive_parachute_alpha {type=number}
 skydive_umbrella_alpha {type=number}
 skydiving_umbrella_active {type=boolean}
 spine_rot_alpha {type=number}
]
